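# Translucent Niagara sprite material for procedural-looking flame particles, built from two tiling noise textures.
# A distortion noise scrolls with particle age to warp the flame-shape noise, with per-particle random UV offsets varying each sprite.
# The noise's green channel drives particle-tinted color with boosted hot spots and inverse eye adaptation.
# It also drives opacity, eroded by remapping against a radial edge mask and age threshold, then particle alpha and optional depth fade.
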Material /NiagaraToolsets/Templates/Materials/M_Flames
graph {
 properties {
  bUsedWithNiagaraSprites=true
  blendMode=BLEND_Translucent
  opacityMaskClipValue=0.1
  pixelDepthOffsetMode=PDOM_Legacy
  usedShadingModels=Default Lit
 }
 nodes [
  n1 = ScalarParameter() {defaultValue=20, parameterName=Color Boost Exponent}
  n2 = Multiply(A=TextureCoordinate(), B=ScalarParameter() {defaultValue=0.3, parameterName=Texture Scale})
  n3 = ParticleRandom()
  n4 = AppendVector(A=@n3, B=Frac(None=Multiply(A=@n3) {constB=16807.3105}))
  n5 = ScalarParameter() {defaultValue=0.3, parameterName=Distortion Amount}
  n6 = TextureCoordinate()
  n7 = ScalarParameter() {defaultValue=0.2, parameterName=Distortion Scale}
  n8 = ParticleRandom()
  n9 = Add(
   A=Frac(None=Multiply(A=@n8) {constB=16807.3105})
   B=Multiply(A=ParticleRelativeTime(), B=ScalarParameter() {defaultValue=0.25, parameterName=Distortion Speed})
  )
  n10 = TextureSampleParameter2D(UVs=Add(A=Multiply(A=@n6, B=@n7) {constB=0.5}, B=AppendVector(A=@n8, B=@n9))) {
   parameterName=Distortion Texture
   samplerType=SAMPLERTYPE_LinearColor
   texture=/NiagaraToolsets/Templates/Textures/T_HermiteNoise_CenteredTiling
  }
  n11 = ComponentMask(None=Subtract(A=@n10.RGB) {constB=0.5}) {g=true, r=true}
  n12 = TextureSampleParameter2D(UVs=Add(A=Add(A=@n2, B=@n4), B=Multiply(A=@n5, B=@n11) {constB=0}) {desc=Apply Distortion}) {
   parameterName=Flame Shape Texture
   samplerType=SAMPLERTYPE_LinearColor
   texture=/NiagaraToolsets/Templates/Textures/T_Hermite2Noise_CenteredTiling
  }
  n13 = Power(Base=@n12.G, Exp=ScalarParameter() {defaultValue=1.5, parameterName=Opacity Exponent})
  n14 = Add(A=@n13, B=Multiply(A=Power(Base=@n12.G, Exp=@n1), B=ScalarParameter() {defaultValue=50, parameterName=Color Boost Gain}))
  n15 = ParticleColor()
  Reroute(None=EyeAdaptationInverse(LightValueInput=Multiply(A=@n14, B=@n15.RGB), AlphaInput=ScalarParameter() {defaultValue=1, parameterName=Eye Adaptation Mix}).EyeAdaptationInverse)
  n16 = Saturate(None=@n13)
  n17 = OneMinus(None=SphereMask(A=TextureCoordinate(), B=Constant() {r=0.5}) {attenuationRadius=0.5, hardnessPercent=70})
  n18 = Constant() {r=1}
  n19 = ParticleRelativeTime()
  n20 = ScalarParameter() {defaultValue=10, parameterName=Age Threshold Exponent}
  n21 = Multiply(
   A=@n15.A
   B=MaterialFunctionCall!(
    Input=MaterialFunctionCall(Input=@n16, Input Low=@n17, Input High=@n18) {desc=Edge Fade, materialFunction=/Engine/Functions/Engine_MaterialFunctions03/Math/RemapValueRangeNormalized}.Result
    Input Low=Power(Base=@n19, Exp=@n20)
    Input High=Constant() {r=1}
   ) {desc=Age Based Threshold, materialFunction=&RemapValueRangeNormalized}.Result
  )
  StaticSwitchParameter(True=DepthFade(Opacity=@n21, FadeDistance=ScalarParameter() {defaultValue=25, parameterName=FadeDistance}), False=@n21) {defaultValue=true, parameterName=Use Depth Fade}
 ]
}
