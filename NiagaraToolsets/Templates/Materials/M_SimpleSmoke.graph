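# Translucent, lit smoke material template for Niagara sprites, using volumetric directional lighting and generated spherical particle normals.
# The Base Color parameter times particle color supplies the RGB tint and an alpha multiplier.
# A scaled tiling smoke texture gets a per-particle UV offset, with the second random axis hashed from Particle Random via multiply and frac.
# That texture plus a soft radial sphere mask is subtracted and saturated, eroding opacity toward sprite edges.
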Material /NiagaraToolsets/Templates/Materials/M_SimpleSmoke
graph {
 properties {
  bAutomaticallySetUsageInEditor=false
  bGenerateSphericalParticleNormals=true
  bUsedWithNiagaraSprites=true
  blendMode=BLEND_Translucent
  translucencyLightingMode=TLM_VolumetricDirectional
  usedShadingModels=Default Lit
 }
 nodes [
  n1 = Multiply(A=VectorParameter() {parameterName=Base Color, defaultValue=<rgba 1 1 1 1>}.RGBA, B=ParticleColor().RGBA)
  ComponentMask(None=@n1) {b=true, g=true, r=true}
  n2 = Multiply(A=TextureCoordinate(), B=ScalarParameter() {defaultValue=0.6, parameterName=Texture Scale}) {constB=0.5}
  n3 = ParticleRandom()
  n4 = AppendVector(A=@n3, B=Frac(None=Multiply(A=@n3) {constB=16807.3105}))
  n5 = TextureCoordinate()
  n6 = Constant() {r=0.5}
  n7 = Add(
   A=TextureSampleParameter2D(UVs=Add(A=@n2, B=@n4)) {parameterName=Tiling Alpha Texture, samplerType=SAMPLERTYPE_Alpha, texture=/NiagaraToolsets/Templates/Textures/T_SmokeTile}.R
   B=Power(Base=SphereMask(A=@n5, B=@n6) {attenuationRadius=0.5, hardnessPercent=0}) {constExponent=0.5}
  )
  Multiply(A=ComponentMask(None=@n1) {a=true}, B=Saturate(None=Subtract(A=@n7)))
 ]
}
