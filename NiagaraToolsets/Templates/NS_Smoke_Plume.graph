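# Niagara template system for a rising, wind-blown smoke plume: one sprite emitter with user-set color, spawn rate and sprite size.
# Particles spawn in a sphere, where vector noise randomizes temperature and alpha, and an alpha threshold culls some.
# Decaying temperature drives upward buoyancy, combined with fading curl noise, wind and drag.
# Sprites rotate, grow over life, and fade in quickly then out.
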NiagaraSystem /NiagaraToolsets/Templates/NS_Smoke_Plume
dependencies {usedDataInterfaces [&NiagaraDataInterfaceCurve], usedDynamicInputs [], usedModules [&SystemState]}
emitters {SmokePlume {
  emitterSpawnScript [
   /SetVariables(Emitter.Buoyancy=1, Emitter.BuoyancyDecay=0.5) {isSetParameters=true}
   /SetVariables(Emitter.WindDirection:Vector3f=<1 0 0>, Emitter.WindStrength=50) {isSetParameters=true}
   /SetVariables(Emitter.BaseSize=100) {isSetParameters=true}
  ]
  emitterUpdateScript [
   &EmitterState()
   &SpawnRate(SpawnRate=$User.Spawn Rate, Use Spawn Probability=false)
  ]
  particleSpawnScript [
   &InitializeParticle(
    Color=$User.Smoke Color
    Lifetime Max=12
    Lifetime Min=3
    Lifetime Mode=::NewEnumerator1
    Mass Max=0.5
    Mass Min=0.25
    Mass Mode=::NewEnumerator3
    Sprite Rotation Mode=::NewEnumerator2
    Sprite Size Mode=::NewEnumerator1
    Uniform Sprite Size Max=&Multiply_Float(A=$User.Sprite Size, B=1.25)
    Uniform Sprite Size Min=&Multiply_Float(A=$User.Sprite Size, B=0.75)
    UsePositionOffset=false
    Write Material Random=true
   )
   &ShapeLocation(
    MaskSpawnGroup=false
    Override Seed=false
    Sphere Radius=&Multiply_Float(A=$Emitter.BaseSize, B=0.5)
    UseOwnerScale=true
   )
   /Niagara/Modules/Update/Utility/VectorNoiseValue(
    Noise Frequency:NiagaraFloat=&Multiply_Float(A=&Divide_Float(A=1, B=$Emitter.BaseSize), B=500)
    Pan Noise=true
    Pan Noise Field:Vector3f=<0 0 0.1>
    Remap Output Value:ENiagara_CurlNoiseRemapType=::NewEnumerator1
    Remap Range Output:Vector2f=<0 1>
   )
   /SetVariables(
    Particles.Temperature:NiagaraFloat=&RemapRange(Input Value To Remap=$Output.VectorNoiseValue.NoiseX, Input Min=0, Input Max=1, Output Min=0.75, Output Max=1)
    Particles.TemperatureDecay:NiagaraFloat=&Multiply_Float(A=$Emitter.BuoyancyDecay, B=$Engine.DeltaTime)
   ) {isSetParameters=true}
   &Color(Color=&MultiplyLinearColor(
     Linear Color=$Particles.Color
     Scale Factor=&MakeFloat4FromVectorAndFloat("Vector (RGB)"=<1 1 1>, "Float (Alpha)"=$Output.VectorNoiseValue.NoiseY)
    ))
   &KillParticles(Kill Particles=&SetBoolByFloatComparison(
     A=&MakeFloatFromLinearColor(LinearColor=$Particles.Color, Channel:ENiagaraFloatFromLinearColorOptions=::NewEnumerator3)
     B=0.01
     Comparison Type=::NewEnumerator6
    ))
  ]
  particleUpdateScript [
   &ParticleState()
   /SetVariables(Particles.Temperature:NiagaraFloat=&Multiply_Float(
     A=$Particles.Temperature
     B=&ClampFloat(Float=&OneMinusFloat(Float=$Particles.TemperatureDecay))
    )) {isSetParameters=true}
   &LinearForce(Force=&Multiply_VectorByFloat(Vector=<0 0 200>, Float=&Multiply_Float(A=$Particles.Temperature, B=$Emitter.Buoyancy)))
   &CurlNoiseForce(
    Bias Noise=false
    Noise Frequency=3.5
    Noise Strength=&FloatFromCurve(
     FloatCurve=DataInterface() {propertyValues {Curve {
        defaultValue=1
        keys [{interpMode=RCIM_Cubic, tangentMode=RCTM_User, time=0, value=1}, {interpMode=RCIM_Cubic, tangentMode=RCTM_User, time=1, value=0}]
       }}}
     Scale Curve=20
    )
    Pan Noise=true
    Pan Noise Field=<0 0 0.5>
   )
   &WindForce(Frequency=2, Wind Speed=$Emitter.WindDirection, Wind Speed Scale=$Emitter.WindStrength)
   &Drag(Use Linear Drag=true, Use Rotational Drag=true)
   &SolveForcesAndVelocity(Clamp Velocity=false, Limit Acceleration=false)
   &SpriteRotationRate(Rotation Rate=&RandomRangeFloat(Minimum=-5, Maximum=5))
   &ScaleColor(
    Scale Alpha=&FloatFromCurve(FloatCurve=DataInterface() {propertyValues {Curve {keys [
         {interpMode=RCIM_Cubic, tangentMode=RCTM_User, time=0, value=0}
         {
          arriveTangent=-0.2248
          arriveTangentWeight=0.0745
          interpMode=RCIM_Cubic
          leaveTangent=-0.2248
          leaveTangentWeight=0.2981
          tangentMode=RCTM_User
          time=0.025
          value=1
         }
         {
          arriveTangent=-1.6181
          arriveTangentWeight=0.3133
          interpMode=RCIM_Cubic
          leaveTangent=-1.6181
          leaveTangentWeight=0.3133
          tangentMode=RCTM_User
          tangentWeightMode=RCTWM_WeightedBoth
          time=1
          value=0
         }
        ]}}})
    ScaleA=true
    ScaleRGB=true
   )
   &ScaleSpriteSize(
    Scale Sprite Size Mode=::NewEnumerator0
    Uniform Scale Factor=&Add_Float(
     A=1
     B=&FloatFromCurve(
      FloatCurve=DataInterface() {propertyValues {Curve {defaultValue=0, keys [{tangentMode=RCTM_User, time=0, value=0}, {tangentMode=RCTM_User, time=1, value=1}]}}}
      Scale Curve=5
     )
    )
   )
  ]
  rendererClasses [&NiagaraSpriteRendererProperties]
  renderers [{rendererClass=&NiagaraSpriteRendererProperties}]
 }}
summary {
 systemName=NS_Smoke_Plume
 emitters [{emitterName=SmokePlume, rendererClasses [&NiagaraSpriteRendererProperties]}]
 userVariables [
  User.Smoke Color {type=LinearColor, description="", defaultValue {struct=LinearColor, value=<rgba 0.3 0.3 0.3 1>}}
  User.Spawn Rate {type=NiagaraFloat, defaultValue=10, description=""}
  User.Sprite Size {type=NiagaraFloat, defaultValue=100, description=""}
 ]
}
